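# Niagara system for the Propulsion Gel vehicle boost cosmetic: sprite gel sprayed from skeletal mesh sockets, with splashes.
# Gel_01 and two-group Gel_02 partition particles across sockets, arc them under gravity and animate subimages.
# Gel_01 writes death positions to a system array, which feeds Splash burst position through ReadDeath.
# VC_Mask inputs retune values for Coro, FrontEnd (split by preview flag) and FN; only Splash has painted-variant colors enabled.
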NiagaraSystem /VehicleCosmetics/Boosts/Boost_PropulsionGel/FX/NS_PropulsionGel
dependencies {
 usedDataInterfaces [
  &NiagaraDataInterfaceCurve
  &NiagaraDataInterfaceArrayInt32
  &NiagaraDataInterfaceVectorCurve
  &NiagaraDataInterfaceSkeletalMesh
  &NiagaraDataInterfaceArrayPosition
  &NiagaraDataInterfaceArrayFloat3
 ]
 usedDynamicInputs [
  /VehicleCosmetics/FX/Niagara/Modules/VC_Mask_LinearColor
  &MakeCustomLinearColorFromBool
  &SelectVectorFromArray
 ]
 usedModules [
  &self:SystemSpawnScript.SystemScriptSource.SystemScriptGraph.NiagaraNodeAssignment_5.SetVariables_#1
  &self:SystemSpawnScript.SystemScriptSource.SystemScriptGraph.NiagaraNodeAssignment_1.SetVariables_#2
  &self:SystemSpawnScript.SystemScriptSource.SystemScriptGraph.NiagaraNodeAssignment_0.SetVariables_#3
  &self:SystemSpawnScript.SystemScriptSource.SystemScriptGraph.NiagaraNodeAssignment_3.SetVariables_#4
  &self:SystemSpawnScript.SystemScriptSource.SystemScriptGraph.NiagaraNodeAssignment_6.SetVariables_#5
  &self:SystemSpawnScript.SystemScriptSource.SystemScriptGraph.NiagaraNodeAssignment_2.SetVariables_#6
  &SystemState
  &self:ReadDeath
  &self:SystemSpawnScript.SystemScriptSource.SystemScriptGraph.NiagaraNodeAssignment_4.SetVariables_#7
 ]
}
emitters {
 NE_VC_Boost_Gel_01 {
  emitterUpdateScript [
   &EmitterState()
   &VC_FilteredSocketCount()
   &SpawnRate(
    SpawnRate=&Multiply_Float(
     A=100
     B=/VehicleCosmetics/FX/Niagara/Modules/VC_Mask_Floats(
      Coro Value:NiagaraFloat=0.5
      FrontEnd Value:NiagaraFloat=&MakeCustomFloatFromBool(Bool=$User.bIsFrontendPreview, True Float=0.2, False Float=2)
      FN Value:NiagaraFloat=0.5
     )
    )
    Use Spawn Probability=false
   )
  ]
  particleSpawnScript [
   &VC_PaintedVariantsOverride() {enabled=false}
   &InitializeParticle(
    Color Mode=::NewEnumerator0
    Lifetime=&VC_Mask_Floats(
     Coro Value:NiagaraFloat=1
     FrontEnd Value:NiagaraFloat=&MakeCustomFloatFromBool(Bool=$User.bIsFrontendPreview, True Float=0.64, False Float=0.53)
     FN Value:NiagaraFloat=0.5
    )
    Sprite Rotation Angle=&RandomRangeFloat(Minimum=0.4, Maximum=0.6)
    Sprite Rotation Mode=::NewEnumerator1
    Sprite Size Max=/VehicleCosmetics/FX/Niagara/Modules/VC_Mask_Vector2(
     Coro Value:Vector2f=<0 0>
     FrontEnd Value:Vector2f=&MakeCustomVector2DFromBool(Bool=$User.bIsFrontendPreview, True Vector2D=<100 100>, False Vector2D=<110 110>)
     FN Value:Vector2f=<0 0>
    )
    Sprite Size Min=&VC_Mask_Vector2(
     Coro Value:Vector2f=<0 0>
     FrontEnd Value:Vector2f=&MakeCustomVector2DFromBool(Bool=$User.bIsFrontendPreview, True Vector2D=<120 120>, False Vector2D=<125 125>)
     FN Value:Vector2f=<0 0>
    )
    Sprite Size Mode=::NewEnumerator2
    UsePositionOffset=false
    Write Material Random=true
   )
   &PartitionParticles(Number Of Partitions=$Emitter.PortCount)
   &SkeletalMeshLocation(
    Bone / Socket Index=$Particles.PartitionParticles.Partition
    Bone / Socket Sampling Mode=::NewEnumerator5
    MaskSpawnGroup=false
    Mesh Sampling Type=::NewEnumerator6
    Sampled Position Offset=&VC_Mask_Vectors(
     Coro Value:Vector3f=<0 0 0>
     FrontEnd Value:Vector3f=&MakeCustomVectorFromBool(Bool=$User.bIsFrontendPreview, True Vector=<0 0 0>, False Vector=<180 0 -30>)
     FN Value:Vector3f=<0 0 0>
    )
    Skeletal Mesh=$System.SkeletalMesh
   )
   &ShapeLocation(MaskSpawnGroup=false, Override Seed=false, Sphere Radius=10, UseOwnerScale=true)
   &AddVelocity(Velocity=&VC_Mask_Vectors(
     Coro Value:Vector3f=&RandomRangeVector(Minimum=<-100 0 300>, Maximum=<-300 1 500>)
     FrontEnd Value:Vector3f=&MakeCustomVectorFromBool(
      Bool=$User.bIsFrontendPreview
      True Vector=&RandomRangeVector(Minimum=<-1000 0 350>, Maximum=<-1200 0 350>)
      False Vector=&RandomRangeVector(Minimum=<-800 0 300>, Maximum=<-300 0 500>)
     )
     FN Value:Vector3f=&RandomRangeVector(Minimum=<-1300 0 800>, Maximum=<-1700 1 500>)
    ))
   /VehicleCosmetics/FX/VC_FXParents/Modules/DLR_ProductPreviews(GameModeID:NiagaraInt32=$System.GameModeID, bPreviewVehicleSpeed:NiagaraBool=$System.bPreviewVehicleSpeed)
  ]
  particleUpdateScript [
   &ParticleState()
   &self:WriteDeath_03(Position Array:NiagaraDataInterfaceArrayPosition=$System.DeathArray) {module=WriteDeath_03}
   &ScaleSpriteSize(Uniform Curve Sprite Scale=DataInterface() {propertyValues {Curve {keys [{tangentMode=RCTM_User, time=0, value=0}, {tangentMode=RCTM_User, time=0.02, value=1}]}}})
   &SpriteRotationRate(Rotation Rate=&RandomRangeFloat(Minimum=-1, Maximum=1))
   &GravityForce(Gravity=&VC_Mask_Vectors(
     Coro Value:Vector3f=<-400 0 0>
     FrontEnd Value:Vector3f=&MakeCustomVectorFromBool(Bool=$User.bIsFrontendPreview, True Vector=<0 0 -1500>, False Vector=<0 0 -1500>)
     FN Value:Vector3f=<-1200 0 0>
    ))
   &InheritSourceMovement(
    Applied Position Scale=&VectorFromFloat(Value=&FloatFromCurve(FloatCurve=DataInterface() {propertyValues {Curve {
         defaultValue=1
         keys [
          {arriveTangent=-1.7885, interpMode=RCIM_Cubic, leaveTangent=-1.7885, tangentMode=RCTM_User, time=0.05, value=1}
          {interpMode=RCIM_Cubic, tangentMode=RCTM_User, time=1, value=0}
         ]
        }}}))
    ApplyForce=false
    ApplyPosition=true
    ApplyPositionFalloff=false
   )
   &SolveForcesAndVelocity(Clamp Velocity=false, Limit Acceleration=false)
   /SetVariables(Particles.SubImageIndex=&FloatFromCurve(FloatCurve=DataInterface() {propertyValues {Curve {keys [<at 0 0>,<at 1 64>]}}})) {isSetParameters=true}
  ]
  rendererClasses [&NiagaraSpriteRendererProperties]
  renderers [{rendererClass=&NiagaraSpriteRendererProperties}]
 }
 NE_VC_Boost_Gel_02 {
  emitterUpdateScript [
   &EmitterState()
   &VC_FilteredSocketCount()
   &SpawnRate(
    SpawnRate=&Multiply_Float(
     A=50
     B=&VC_Mask_Floats(
      Coro Value:NiagaraFloat=1
      FrontEnd Value:NiagaraFloat=&MakeCustomFloatFromBool(Bool=$User.bIsFrontendPreview, True Float=0.2, False Float=1)
      FN Value:NiagaraFloat=0.5
     )
    )
    Use Spawn Probability=false
   )
   &SpawnRate(
    SpawnGroup=1
    SpawnRate=&Multiply_Float(
     A=100
     B=&VC_Mask_Floats(
      Coro Value:NiagaraFloat=1
      FrontEnd Value:NiagaraFloat=&MakeCustomFloatFromBool(Bool=$User.bIsFrontendPreview, True Float=0.35, False Float=0.75)
      FN Value:NiagaraFloat=0.5
     )
    )
    Use Spawn Probability=false
   ) {module=SpawnRate001}
  ]
  particleSpawnScript [
   &VC_PaintedVariantsOverride() {enabled=false}
   &InitializeParticle(
    Color Mode=::NewEnumerator0
    Lifetime Max=&MergeHelper_FloatSwitch(
     0=&VC_Mask_Floats(
      Coro Value:NiagaraFloat=1
      FrontEnd Value:NiagaraFloat=&MakeCustomFloatFromBool(Bool=$User.bIsFrontendPreview, True Float=0.64, False Float=0.5)
      FN Value:NiagaraFloat=0.5
     )
     1=&VC_Mask_Floats(
      Coro Value:NiagaraFloat=1
      FrontEnd Value:NiagaraFloat=&MakeCustomFloatFromBool(Bool=$User.bIsFrontendPreview, True Float=1.3, False Float=1.2)
      FN Value:NiagaraFloat=0.5
     )
     2=?
     3=?
     Index=&SpawnGroup()
     Num=::NewEnumerator0
    )
    Lifetime Min:NiagaraFloat=&MergeHelper_FloatSwitch(
     0=&VC_Mask_Floats(
      Coro Value:NiagaraFloat=1
      FrontEnd Value:NiagaraFloat=&MakeCustomFloatFromBool(Bool=$User.bIsFrontendPreview, True Float=0.64, False Float=0.53)
      FN Value:NiagaraFloat=0.5
     )
     1=&VC_Mask_Floats(
      Coro Value:NiagaraFloat=1
      FrontEnd Value:NiagaraFloat=&MakeCustomFloatFromBool(Bool=$User.bIsFrontendPreview, True Float=1.1, False Float=1.3)
      FN Value:NiagaraFloat=0.5
     )
     2=?
     3=?
     Index=&SpawnGroup()
     Num=::NewEnumerator0
    )
    Lifetime Mode=::NewEnumerator1
    Sprite Rotation Angle=&MergeHelper_FloatSwitch(
     0=&RandomRangeFloat(Minimum=0.4, Maximum=0.6)
     1=&RandomRangeFloat(Minimum=0.8, Maximum=0.2)
     2=?
     3=?
     Index=&SpawnGroup()
     Num=::NewEnumerator0
    )
    Sprite Rotation Mode=::NewEnumerator1
    Sprite Size Mode=::NewEnumerator1
    Uniform Sprite Size Max=70
    Uniform Sprite Size Min=30
    UsePositionOffset=false
    Write Material Random=true
   )
   /SetVariables(Particles.VisibilityTag=&SelectIntFromArray(Direct Array Index=&SpawnGroup(), Int Selection Array=DataInterface() {propertyValues {IntData [0,1]}})) {isSetParameters=true}
   &PartitionParticles(Number Of Partitions=$Emitter.PortCount)
   &SkeletalMeshLocation(
    Bone / Socket Index=$Particles.PartitionParticles.Partition
    Bone / Socket Sampling Mode=::NewEnumerator5
    MaskSpawnGroup=false
    Mesh Sampling Type=::NewEnumerator6
    Sampled Position Offset=&VC_Mask_Vectors(
     Coro Value:Vector3f=<0 0 0>
     FrontEnd Value:Vector3f=&MakeCustomVectorFromBool(Bool=$User.bIsFrontendPreview, True Vector=<0 0 0>, False Vector=<190 0 -30>)
     FN Value:Vector3f=<0 0 0>
    )
    Skeletal Mesh=$System.SkeletalMesh
   )
   &ShapeLocation(
    MaskSpawnGroup=false
    Override Seed=false
    Sphere Radius=&MergeHelper_FloatSwitch(
     0=10
     1=&VC_Mask_Floats(
      Coro Value:NiagaraFloat=1
      FrontEnd Value:NiagaraFloat=&MakeCustomFloatFromBool(Bool=$User.bIsFrontendPreview, True Float=30, False Float=10)
      FN Value:NiagaraFloat=0.5
     )
     2=?
     3=?
     Index=&SpawnGroup()
     Num=::NewEnumerator0
    )
    UseOwnerScale=true
   )
   &AddVelocity(Velocity=&VC_Mask_Vectors(
     Coro Value:Vector3f=&RandomRangeVector(Minimum=<-1400 0 800>, Maximum=<-1700 1 500>)
     FrontEnd Value:Vector3f=&RandomRangeVector(
      Minimum=&MergeHelper_VectorSwitch(
       0=&MakeCustomVectorFromBool(Bool=$User.bIsFrontendPreview, True Vector=<-1000 0 350>, False Vector=<-100 0 300>)
       1=<0 0 0>
       2=?
       3=?
       Index=&SpawnGroup()
       Num=::NewEnumerator0
      )
      Maximum=&MergeHelper_VectorSwitch(
       0=&MakeCustomVectorFromBool(Bool=$User.bIsFrontendPreview, True Vector=<-1200 0 350>, False Vector=<-300 0 500>)
       1=<0 0 0>
       2=?
       3=?
       Index=&SpawnGroup()
       Num=::NewEnumerator0
      )
     )
     FN Value:Vector3f=&RandomRangeVector(Minimum=<-1300 0 800>, Maximum=<-1700 1 500>)
    ))
   &AddVelocity(
    Cone Angle=&VC_Mask_Floats(
     Coro Value:NiagaraFloat=1
     FrontEnd Value:NiagaraFloat=&MakeCustomFloatFromBool(Bool=$User.bIsFrontendPreview, True Float=30, False Float=60)
     FN Value:NiagaraFloat=0.5
    )
    Cone Axis=&VC_Mask_Vectors(
     Coro Value:Vector3f=<-400 0 0>
     FrontEnd Value:Vector3f=&MakeCustomVectorFromBool(Bool=$User.bIsFrontendPreview, True Vector=<-1 0 0.4>, False Vector=<-1 0 0.7>)
     FN Value:Vector3f=<-1200 0 0>
    )
    Module.UseConeFalloff=false
    Module.UseVelDistribution=false
    Override Seed=false
    Velocity Mode=::NewEnumerator2
    Velocity Speed=&MergeHelper_FloatSwitch(
     0=0
     1=&VC_Mask_Floats(
      Coro Value:NiagaraFloat=1
      FrontEnd Value:NiagaraFloat=&MakeCustomFloatFromBool(
       Bool=$User.bIsFrontendPreview
       True Float=&RandomRangeFloat(Minimum=900, Maximum=1100)
       False Float=&RandomRangeFloat(Minimum=150, Maximum=100)
      )
      FN Value:NiagaraFloat=0.5
     )
     2=?
     3=?
     Index=&SpawnGroup()
     Num=::NewEnumerator0
    )
   ) {module=AddVelocity001}
   &DLR_ProductPreviews(GameModeID:NiagaraInt32=$System.GameModeID, bPreviewVehicleSpeed:NiagaraBool=$System.bPreviewVehicleSpeed)
  ]
  particleUpdateScript [
   &ParticleState()
   &AccelerationForce(Acceleration=&MergeHelper_VectorSwitch(
     0=<0 0 0>
     1=&VC_Mask_Vectors(
      Coro Value:Vector3f=<0 0 0>
      FrontEnd Value:Vector3f=&MakeCustomVectorFromBool(
       Bool=$User.bIsFrontendPreview
       True Vector=<0 0 0>
       False Vector=&VectorFromCurve(
        VectorCurve=DataInterface() {propertyValues {XCurve {keys [<at 0 1>,<at 0.2 0>]}, YCurve {keys [<at 0 1>,<at 0.2 0>]}, ZCurve {keys [<at 0 0>,<at 0.12 1>,<at 0.32 0>]}}}
        Scale Curve=&RandomRangeVector(Minimum=<3600 3600 5800>, Maximum=<-3600 -3600 3500>)
       )
      )
      FN Value:Vector3f=<0 0 0>
     )
     2=?
     3=?
     Index=&SpawnGroup()
     Num=::NewEnumerator0
    ))
   &ScaleSpriteSize(
    Scale Factor=&MergeHelper_Vector2Switch(
     0=&VC_Mask_Vector2(
      Coro Value:Vector2f=<1 1>
      FrontEnd Value:Vector2f=&MakeCustomVector2DFromBool(
       Bool=$User.bIsFrontendPreview
       True Vector2D=<1.5 1.5>
       False Vector2D=&RandomRangeVector2D(Minimum=<1.8 1.8>, Maximum=<1.6 1.6>)
      )
      FN Value:Vector2f=<1 1>
     )
     1=&VC_Mask_Vector2(
      Coro Value:Vector2f=<1 1>
      FrontEnd Value:Vector2f=&MakeCustomVector2DFromBool(Bool=$User.bIsFrontendPreview, True Vector2D=<0.7 0.7>, False Vector2D=<1 1>)
      FN Value:Vector2f=<1 1>
     )
     2=?
     3=?
     4=?
     Index=&SpawnGroup()
     Num=::NewEnumerator0
    )
    Scale Sprite Size Mode=::NewEnumerator2
   )
   &SpriteRotationRate(Rotation Rate=&RandomRangeFloat(Minimum=-1, Maximum=1))
   &GravityForce(Gravity=&MergeHelper_VectorSwitch(
     0=&VC_Mask_Vectors(
      Coro Value:Vector3f=<-400 0 0>
      FrontEnd Value:Vector3f=&MakeCustomVectorFromBool(Bool=$User.bIsFrontendPreview, True Vector=<0 0 -1500>, False Vector=<0 0 -2000>)
      FN Value:Vector3f=<-1200 0 0>
     )
     1=&VC_Mask_Vectors(
      Coro Value:Vector3f=<-400 0 0>
      FrontEnd Value:Vector3f=&MakeCustomVectorFromBool(Bool=$User.bIsFrontendPreview, True Vector=<0 0 -900>, False Vector=<0 0 -1500>)
      FN Value:Vector3f=<-1200 0 0>
     )
     2=?
     3=?
     Index=&SpawnGroup()
     Num=::NewEnumerator0
    ))
   &InheritSourceMovement(
    Applied Position Scale=&VectorFromFloat(Value=&FloatFromCurve(FloatCurve=DataInterface() {propertyValues {Curve {
         defaultValue=1
         keys [
          {arriveTangent=-1.7885, interpMode=RCIM_Cubic, leaveTangent=-1.7885, tangentMode=RCTM_User, time=0.05, value=1}
          {interpMode=RCIM_Cubic, tangentMode=RCTM_User, time=1, value=0}
         ]
        }}}))
    ApplyForce=false
    ApplyPosition=true
    ApplyPositionFalloff=false
   )
   &SolveForcesAndVelocity(Clamp Velocity=false, Limit Acceleration=false)
   /SetVariables(Particles.SubImageIndex=&MergeHelper_FloatSwitch(
     0=&FloatFromCurve(FloatCurve=DataInterface() {propertyValues {Curve {keys [<at 0 0>,<at 1 64>]}}})
     1=&FloatFromCurve(FloatCurve=DataInterface() {propertyValues {Curve {keys [<at 0 0>,<at 1 64>]}}})
     2=?
     3=?
     Index=&SpawnGroup()
     Num=::NewEnumerator0
    )) {isSetParameters=true}
  ]
  rendererClasses [&NiagaraSpriteRendererProperties]
  renderers [
   {rendererClass=&NiagaraSpriteRendererProperties}
   {rendererClass=&NiagaraSpriteRendererProperties, rendererIndex=1}
  ]
 }
 NE_VC_Boost_Splash {
  emitterUpdateScript [
   &EmitterState()
   &SpawnPerFrame(
    Spawn:NiagaraBool=$System.DeathSpawn
    Spawn Count:NiagaraInt32=&MakeIntFromFloat(Float=&Multiply_Float(
      A=1
      B=&VC_Mask_Floats(
       Coro Value:NiagaraFloat=1
       FrontEnd Value:NiagaraFloat=&MakeCustomFloatFromBool(Bool=$User.bIsFrontendPreview, True Float=3, False Float=4)
       FN Value:NiagaraFloat=0.5
      )
     ))
    Use Spawn Probability=false
   )
   &VC_FilteredSocketCount() {enabled=false}
   &SpawnRate(
    SpawnRate=&Multiply_Float(
     A=200
     B=&VC_Mask_Floats(Coro Value:NiagaraFloat=1, FrontEnd Value:NiagaraFloat=1, FN Value:NiagaraFloat=0.5)
    )
    Use Spawn Probability=false
   ) {enabled=false}
  ]
  particleSpawnScript [
   &VC_PaintedVariantsOverride(
    Black=<rgba 0 0 0 1>
    BurntSienna=<rgba 1.4 0.3939 0.124 1>
    Cobalt=<rgba 0 0.148 8 1>
    Crimson=<rgba 5 0 0.1355 1>
    DefaultColor=<rgba 6 6 6 1>
    ForestGreen=<rgba 0 5 0.2526 1>
    Grey=<rgba 0.1927 0.1927 0.1927 1>
    Lime=<rgba 3 6 0 1>
    Orange=<rgba 5 0.6667 0 1>
    Pink=<rgba 2 0.2396 1.3598 1>
    Purple=<rgba 0.6156 0 1.5 1>
    SkyBlue=<rgba 0 3.3269 5 1>
    TitaniumWhite=<rgba 4 4 4 1>
   )
   &InitializeParticle(
    Color=<rgba 0.3 0.0811 0 1>
    Lifetime=&RandomRangeFloat(Minimum=0.4, Maximum=0.3)
    Mass Mode=::NewEnumerator3
    Position=$System.DeathPosition
    Position Mode=::NewEnumerator1
    Sprite Rotation Mode=::NewEnumerator2
    Sprite Size Mode=::NewEnumerator1
    Uniform Sprite Size Max=&VC_Mask_Floats(
     Coro Value:NiagaraFloat=1
     FrontEnd Value:NiagaraFloat=&MakeCustomFloatFromBool(Bool=$User.bIsFrontendPreview, True Float=400, False Float=600)
     FN Value:NiagaraFloat=0.5
    )
    Uniform Sprite Size Min=&VC_Mask_Floats(
     Coro Value:NiagaraFloat=1
     FrontEnd Value:NiagaraFloat=&MakeCustomFloatFromBool(Bool=$User.bIsFrontendPreview, True Float=200, False Float=1000)
     FN Value:NiagaraFloat=0.5
    )
    UsePositionOffset=false
    Write Material Random=true
   )
   &PartitionParticles(Number Of Partitions=$Emitter.PortCount) {enabled=false}
   &SkeletalMeshLocation(
    Bone / Socket Index=$Particles.PartitionParticles.Partition
    Bone / Socket Sampling Mode=::NewEnumerator5
    MaskSpawnGroup=false
    Mesh Sampling Type=::NewEnumerator6
    Skeletal Mesh=$System.SkeletalMesh
   ) {enabled=false}
   &AddVelocity(
    Cone Angle=140
    Cone Axis=<1 0 0.8>
    Module.UseConeFalloff=false
    Module.UseVelDistribution=false
    Override Seed=false
    Velocity Mode=::NewEnumerator2
    Velocity Speed=&RandomRangeFloat(Minimum=500, Maximum=800)
   )
   &ShapeLocation(
    MaskSpawnGroup=false
    Offset=&MakeVector(
     X=0
     Y=&UniformAOrBFloat(A=0, B=0, Evaluation Type=::NewEnumerator0, Recalculate Random Each Loop=true, Override Seed=false)
     Z=&VC_Mask_Floats(
      Coro Value:NiagaraFloat=1
      FrontEnd Value:NiagaraFloat=&MakeCustomFloatFromBool(Bool=$User.bIsFrontendPreview, True Float=30, False Float=-30)
      FN Value:NiagaraFloat=0.5
     )
    )
    Offset Coordinate Space=::Local
    Offset Mode=::NewEnumerator0
    Override Seed=false
    Sphere Radius=10
    UseOwnerScale=true
   )
   &DLR_ProductPreviews(GameModeID:NiagaraInt32=$System.GameModeID, bPreviewVehicleSpeed:NiagaraBool=$System.bPreviewVehicleSpeed)
  ]
  particleUpdateScript [
   &ParticleState()
   &ScaleSpriteSize(Uniform Curve Sprite Scale=DataInterface() {propertyValues {Curve {keys [{tangentMode=RCTM_User, time=0, value=0.3}, {tangentMode=RCTM_User, time=1, value=1}]}}})
   &SpriteRotationRate(Rotation Rate=&RandomRangeFloat(Minimum=1, Maximum=-1))
   &GravityForce(Gravity=&RandomRangeVector(Minimum=<0 0 -1000>, Maximum=<1 1 -2000>))
   &InheritSourceMovement(
    Applied Position Scale=&VectorFromFloat(Value=&FloatFromCurve(FloatCurve=DataInterface() {propertyValues {Curve {
         defaultValue=1
         keys [
          {arriveTangent=-1.7885, interpMode=RCIM_Cubic, leaveTangent=-1.7885, tangentMode=RCTM_User, time=0.05, value=1}
          {interpMode=RCIM_Cubic, tangentMode=RCTM_User, time=1, value=0}
         ]
        }}}))
    ApplyForce=false
    ApplyPosition=true
    ApplyPositionFalloff=false
   )
   &SolveForcesAndVelocity(Clamp Velocity=false, Limit Acceleration=false)
   &DynamicMaterialParameters(
    Index 0 Param 2=1
    Index 0 Param 4=&FloatFromCurve(FloatCurve=DataInterface() {propertyValues {Curve {keys [{interpMode=RCIM_Cubic, tangentMode=RCTM_Break, time=0, value=0.1}, {interpMode=RCIM_Cubic, tangentMode=RCTM_Break, time=1, value=0.7}]}}})
    Module.Index0Float0Write=true
    Module.Index0Float1Write=true
    Module.Index0Float2Write=true
    Module.Index0Float3Write=true
   )
  ]
  rendererClasses [&NiagaraSpriteRendererProperties]
  renderers [{rendererClass=&NiagaraSpriteRendererProperties}]
 }
}
summary {
 systemName=NS_PropulsionGel
 emitters [
  {emitterName=NE_VC_Boost_Gel_01, rendererClasses [&NiagaraSpriteRendererProperties]}
  {emitterName=NE_VC_Boost_Gel_02, rendererClasses [&NiagaraSpriteRendererProperties]}
  {emitterName=NE_VC_Boost_Splash, rendererClasses [&NiagaraSpriteRendererProperties]}
 ]
 userVariables [
  User.bIsFrontendPreview {type=NiagaraBool}
  User.bIsPainted {type=NiagaraBool}
  User.blsIsCoroColor {type=NiagaraBool}
  User.bPreviewVehicleSpeed {type=NiagaraBool}
  User.ColorVariantID {type=NiagaraInt32}
  User.GameModeID {type=NiagaraInt32}
  User.PaintedColor {type=LinearColor, description="", defaultValue {struct=LinearColor, value=<rgba 1 1 1 1>}}
  User.PaintedScalar {type=NiagaraFloat, defaultValue=1, description=""}
  User.PortCount {type=NiagaraInt32, defaultValue=2}
 ]
}
